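# Niagara effect for the VowelSleep emote that draws a short-lived ribbon trail behind the moving component, likely a ball going by the asset name.
# The Trail emitter spawns points by distance traveled after a System.Delay loop delay, with its time-based spawn rate disabled.
# Ribbon color is dimmed, and an alpha curve fades it in quickly, then out.
# A drifting SubUV sprite Smoke emitter is present but disabled.
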NiagaraSystem /BRCosmetics/Effects/Fort_Effects/Emotes/VowelSleep/NS_Emote_VowelSleep_BallITrail
dependencies {usedDataInterfaces [&NiagaraDataInterfaceCurve], usedDynamicInputs [], usedModules [&SystemState]}
emitters {
 Smoke {
  bEnabled=false
  emitterUpdateScript [
   &EmitterState(Life Cycle Mode=::NewEnumerator1, Loop Behavior=::NewEnumerator1, Loop Duration=1, UseLoopDelay=false)
   &SpawnBurst_Instantaneous(Use Spawn Probability=false, UseLoopCountLimit=false)
   &SpawnRate(SpawnRate=10, Use Spawn Probability=false)
  ]
  particleSpawnScript [
   &InitializeParticle(
    Color=<rgba 0.4727 0.624 0.8177 0.2>
    Lifetime Max=1.5
    Lifetime Mode=::NewEnumerator1
    Sprite Rotation Mode=::NewEnumerator2
    Sprite Size Mode=::NewEnumerator1
    Uniform Sprite Size Max=30
    Uniform Sprite Size Min=10
    UsePositionOffset=false
    Write Material Random=true
   )
   &AddVelocity(
    Cone Angle=20
    Cone Axis=<0 0 1>
    Module.UseConeFalloff=false
    Module.UseVelDistribution=false
    Override Seed=false
    Rotation Coordinate Space=::World
    Velocity Mode=::NewEnumerator2
    Velocity Speed=40
   )
  ]
  particleUpdateScript [
   &ParticleState()
   &ScaleSpriteSize(Uniform Curve Sprite Scale=DataInterface() {propertyValues {Curve {
       defaultValue=0.3
       keys [
        {arriveTangent=0.1319, interpMode=RCIM_Cubic, leaveTangent=0.1319, tangentMode=RCTM_User, time=-0.0031, value=1.2}
        {interpMode=RCIM_Cubic, tangentMode=RCTM_User, time=1, value=0.7}
       ]
      }}})
   &Drag(Drag=1.5, Use Linear Drag=true, Use Rotational Drag=true)
   &SolveForcesAndVelocity(Clamp Velocity=false, Limit Acceleration=false)
   &ScaleColor(
    Scale Alpha=&FloatFromCurve(FloatCurve=DataInterface() {propertyValues {Curve {keys [
         <at 0 0 RCIM_Cubic>
         {arriveTangent=-1.7111, interpMode=RCIM_Cubic, leaveTangent=-1.7111, tangentMode=RCTM_User, time=0.3, value=1}
         <at 1 0>
        ]}}})
    ScaleA=true
    ScaleRGB=true
   )
   &SubUVAnimation()
  ]
  rendererClasses [&NiagaraSpriteRendererProperties]
  renderers [{rendererClass=&NiagaraSpriteRendererProperties}]
 }
 Trail {
  emitterUpdateScript [
   &EmitterState(
    Delay First Loop Only=true
    Inactive Response=::NewEnumerator1
    Life Cycle Mode=::NewEnumerator1
    Loop Delay=$System.Delay
    Loop Duration=1
    UseLoopDelay=true
   )
   &SpawnPerUnit(
    Max Movement Threshold=10000
    Module.Use Max Movement Threshold=true
    Module.Use Movement Tolerance=false
    Spawn Spacing=10
    Use Spawn Probability=false
   )
   &SpawnRate(SpawnRate=60, Use Spawn Probability=false) {enabled=false}
  ]
  particleSpawnScript [&InitializeParticle(Lifetime=0.3, Ribbon Width=15, Ribbon Width Mode=::NewEnumerator1, UsePositionOffset=false, Write Material Random=true)]
  particleUpdateScript [
   &ParticleState()
   &ScaleColor(
    Scale Alpha=&FloatFromCurve(FloatCurve=DataInterface() {propertyValues {Curve {keys [
         <at 0 0 RCIM_Cubic>
         {arriveTangent=-0.1027, interpMode=RCIM_Cubic, leaveTangent=-0.1027, tangentMode=RCTM_User, time=0.1183, value=0.6118}
         {arriveTangent=-0.6939, interpMode=RCIM_Cubic, leaveTangent=-0.6939, time=0.3762, value=0.3007}
         <at 1 0 RCIM_Cubic>
        ]}}})
    Scale RGB=&Multiply_VectorByFloat(Vector=<1 1 1>, Float=0.5)
    ScaleA=true
    ScaleRGB=true
   )
   &SolveForcesAndVelocity(Clamp Velocity=false, Limit Acceleration=false)
  ]
  rendererClasses [&NiagaraRibbonRendererProperties]
  renderers [{rendererClass=&NiagaraRibbonRendererProperties}]
 }
}
summary {
 systemName=NS_Emote_VowelSleep_BallITrail
 emitters [
  {bEnabled=false, emitterName=Smoke, rendererClasses [&NiagaraSpriteRendererProperties]}
  {emitterName=Trail, rendererClasses [&NiagaraRibbonRendererProperties]}
 ]
 userVariables []
}
