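# Niagara sprite effect for the ball idle of the VowelSleep emote, offering three looks chosen by User.CustomFloatType.
# That value indexes a one-hot float array in each emitter's spawn probability, so only the matching emitter spawns.
# Type 0 gives pale blue rising smoke, 1 dark smoke with a hot orange start, and 2 a water-like splash burst.
# Dynamic material parameters animate the latter two, and the dark smoke's alpha varies by platform.
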NiagaraSystem /BRCosmetics/Effects/Fort_Effects/Emotes/VowelSleep/NS_Emote_VowelSleep_BallIdle
dependencies {
 usedDataInterfaces [
  &NiagaraDataInterfaceArrayFloat
  &NiagaraDataInterfaceCurve
  &NiagaraDataInterfaceColorCurve
  &NiagaraDataInterfacePlatformSet
  &NiagaraDataInterfaceVector2DCurve
 ]
 usedDynamicInputs []
 usedModules [&SystemState]
}
emitters {
 Smoke {
  emitterUpdateScript [
   &EmitterState(Life Cycle Mode=::NewEnumerator1, Loop Behavior=::NewEnumerator1, Loop Duration=1, UseLoopDelay=false)
   &SpawnRate(
    Spawn Probability=&SelectFloatFromArray(
     Direct Array Index=&MakeIntFromFloat(Float=$User.CustomFloatType)
     Float Selection Array=DataInterface() {propertyValues {FloatData [1,0,0]}}
    )
    SpawnRate=10
    Use Spawn Probability=true
   )
  ]
  particleSpawnScript [
   &InitializeParticle(
    Color=<rgba 0.4727 0.624 0.8177 0.2>
    Lifetime Max=1.5
    Lifetime Mode=::NewEnumerator1
    Sprite Rotation Mode=::NewEnumerator2
    Sprite Size Mode=::NewEnumerator1
    Uniform Sprite Size Max=30
    Uniform Sprite Size Min=10
    UsePositionOffset=false
    Write Material Random=true
   )
   &AddVelocity(
    Cone Angle=20
    Cone Axis=<0 0 1>
    Module.UseConeFalloff=false
    Module.UseVelDistribution=false
    Override Seed=false
    Rotation Coordinate Space=::World
    Velocity Mode=::NewEnumerator2
    Velocity Speed=40
   )
  ]
  particleUpdateScript [
   &ParticleState()
   &ScaleSpriteSize(Uniform Curve Sprite Scale=DataInterface() {propertyValues {Curve {
       defaultValue=0.3
       keys [
        {arriveTangent=0.1319, interpMode=RCIM_Cubic, leaveTangent=0.1319, tangentMode=RCTM_User, time=-0.0031, value=1.2}
        {interpMode=RCIM_Cubic, tangentMode=RCTM_User, time=1, value=0.7}
       ]
      }}})
   &Drag(Drag=1.5, Use Linear Drag=true, Use Rotational Drag=true)
   &SolveForcesAndVelocity(Clamp Velocity=false, Limit Acceleration=false)
   &ScaleColor(
    Scale Alpha=&FloatFromCurve(FloatCurve=DataInterface() {propertyValues {Curve {keys [
         <at 0 0 RCIM_Cubic>
         {arriveTangent=-1.7111, interpMode=RCIM_Cubic, leaveTangent=-1.7111, tangentMode=RCTM_User, time=0.3, value=1}
         <at 1 0>
        ]}}})
    ScaleA=true
    ScaleRGB=true
   )
   &SubUVAnimation()
  ]
  rendererClasses [&NiagaraSpriteRendererProperties]
  renderers [{rendererClass=&NiagaraSpriteRendererProperties}]
 }
 SmokeLargeLocalSpace {
  emitterUpdateScript [
   &EmitterState(Life Cycle Mode=::NewEnumerator1, Loop Behavior=::NewEnumerator1, Loop Delay=0.5, Loop Duration=1, UseLoopDelay=false)
   &SpawnRate(
    Spawn Probability=&SelectFloatFromArray(
     Direct Array Index=&MakeIntFromFloat(Float=$User.CustomFloatType)
     Float Selection Array=DataInterface() {propertyValues {FloatData [0,1,0]}}
    )
    SpawnRate=10
    Use Spawn Probability=true
   )
  ]
  particleSpawnScript [
   &InitializeParticle(
    Color=<rgba 0.0729 0.0729 0.0729 1>
    Lifetime Max=1.5
    Lifetime Mode=::NewEnumerator1
    Sprite Size Mode=::NewEnumerator1
    Uniform Sprite Size Max=22
    Uniform Sprite Size Min=15
    UsePositionOffset=false
    Write Material Random=true
   )
   &CameraOffset()
   &AddVelocity(
    Cone Angle=20
    Cone Axis=<0 0 1>
    Module.UseConeFalloff=false
    Module.UseVelDistribution=false
    Override Seed=false
    Rotation Coordinate Space=::World
    Velocity Mode=::NewEnumerator2
    Velocity Speed=40
   ) {module=AddVelocity001}
  ]
  particleUpdateScript [
   &Drag(Drag=1.5, Use Linear Drag=true, Use Rotational Drag=true)
   &SolveForcesAndVelocity(Clamp Velocity=false, Limit Acceleration=false)
   &ParticleState()
   &ScaleSpriteSize(
    Uniform Curve Scale=&RandomRangeFloat(Minimum=0.8, Maximum=1.1)
    Uniform Curve Sprite Scale=DataInterface() {propertyValues {Curve {keys [{arriveTangent=0.743, interpMode=RCIM_Cubic, leaveTangent=0.743, tangentMode=RCTM_User, time=0, value=0.8}, <at 1 1.2 RCIM_Cubic>]}}}
   )
   &DynamicMaterialParameters(
    Index 0 Param 1=&Add_Float(
     A=&Multiply_Float(A=$Engine.Time, B=0.5)
     B=&RandomRangeFloat(Minimum=0, Maximum=1)
    )
    Index 0 Param 3=&FloatFromCurve(FloatCurve=DataInterface() {propertyValues {Curve {keys [<at 0 0.1>,<at 0.4 0>,<at 0.4 0>]}}})
    Index 0 Param 4=1
    Index 1 Param 1=0.8
    Index 1 Param 2=&RandomRangeFloat(Minimum=0, Maximum=1)
    Index 1 Param 3=1
    Index 1 Param 4=1.5
    Index 2 Param 1=1
    Index 2 Param 2=&RandomRangeFloat(Minimum=0, Maximum=1)
    Index 2 Param 3=&FloatFromCurve(FloatCurve=DataInterface() {propertyValues {Curve {keys [<at 0 0 RCIM_Cubic>, {arriveTangent=2.1666, tangentMode=RCTM_User, time=1, value=1}]}}})
    Index 2 Param 4=1
    Module.Param0WriteEnabled=true
    Module.Param0WriteEnabled1=true
    Module.Param0WriteEnabled2=true
    Module.Param1WriteEnabled=true
    Module.Param1WriteEnabled1=true
    Module.Param1WriteEnabled2=true
    Module.Param2WriteEnabled=true
    Module.Param2WriteEnabled1=true
    Module.Param2WriteEnabled2=true
    Module.Param3WriteEnabled=true
    Module.Param3WriteEnabled1=true
    Module.Param3WriteEnabled2=true
    Write Parameter Index 1=true
    Write Parameter Index 2=true
   )
   &ScaleColor(
    Scale Alpha=&FloatFromCurve(
     FloatCurve=DataInterface() {propertyValues {Curve {keys [
         {interpMode=RCIM_Cubic, tangentMode=RCTM_User, time=0, value=0}
         <at 0.2 0.8 RCIM_Cubic>
         {interpMode=RCIM_Cubic, tangentMode=RCTM_User, time=1, value=0}
        ]}}}
     Scale Curve=&FN_PickFloatDependOnPlatform(TargetPlatformActive=1, TargetPlatformNoActive=1.5)
    )
    Scale RGB=&MakeVectorFromLinearColorRGB(Color=&ColorFromCurve(ColorCurve=DataInterface() {propertyValues {
        AlphaCurve {keys []}
        BlueCurve {keys [<at 0.1061 0.0066>,<at 0.213 0.5312 RCIM_Cubic>]}
        GreenCurve {keys [<at 0.1061 0.8388>,<at 0.213 0.5312 RCIM_Cubic>]}
        RedCurve {keys [<at 0.1061 5>,<at 0.213 0.5312 RCIM_Cubic>]}
       }}))
    ScaleA=true
    ScaleRGB=true
   )
  ]
  rendererClasses [&NiagaraSpriteRendererProperties]
  renderers [
   {rendererClass=&NiagaraSpriteRendererProperties}
   {rendererClass=&NiagaraSpriteRendererProperties, rendererIndex=1}
  ]
 }
 Splash_Hard {
  emitterUpdateScript [
   &EmitterState(Life Cycle Mode=::NewEnumerator1, Loop Behavior=::NewEnumerator1, Loop Duration=0.15, UseLoopDelay=true)
   &SpawnBurst_Instantaneous(
    Spawn Count=8
    Spawn Probability=&SelectFloatFromArray(
     Direct Array Index=&MakeIntFromFloat(Float=$User.CustomFloatType)
     Float Selection Array=DataInterface() {propertyValues {FloatData [0,0,1]}}
    )
    Use Spawn Probability=true
    UseLoopCountLimit=false
   )
  ]
  particleSpawnScript [
   &InitializeParticle(
    Color=&RandomRangeLinearColor(Minimum=<rgba 1 1 1 1>, Maximum=<rgba 0.133 0.4243 0.6337 1>)
    Lifetime Max=1
    Lifetime Min=0.7
    Lifetime Mode=::NewEnumerator1
    Position Offset Coordinate Space=::Local
    Sprite Rotation Mode=::NewEnumerator2
    Sprite Size Max=<100 150>
    Sprite Size Min=<80 200>
    Sprite Size Mode=::NewEnumerator2
    Sprite UV Mode=::NewEnumerator3
    UsePositionOffset=true
    Write Material Random=true
   )
   &PartitionParticles(Number Of Partitions=1, Particles Per Partition=7, Partition Mode:ENiagara_PartitionMode=::NewEnumerator1)
   &SubUVAnimation(End Frame=3, SubUV Animation Mode:ENiagara_SubUVLookupMode=::NewEnumerator2)
   &ShapeLocation(
    MaskSpawnGroup=false
    Offset=<0 0 10>
    Offset Mode=::NewEnumerator0
    Override Seed=false
    Ring / Disc Distribution Mode:ENiagaraSphereDistributionMode=::NewEnumerator1
    Ring Radius=3
    Shape Primitive=::NewEnumerator4
    U Position=&MaskFloatBySpawnGroup(
     Masked Float=$Particles.PartitionParticles.Normalized Index In Partition
     Passthrough Float=&RandomRangeFloat(Minimum=0, Maximum=1)
    )
    UseOwnerScale=true
   )
   &AddVelocity(
    Module.Use Falloff Distance=false
    Origin Offset=<0 0 -50>
    Override Seed=false
    Velocity Mode=::NewEnumerator1
    Velocity Origin=$Particles.ShapeLocation.ShapePosition
    Velocity Speed=&Multiply_Float(A=&RandomRangeFloat(Minimum=500, Maximum=600), B=0.13)
   )
   &AddVelocity(
    Module.Use Falloff Distance=false
    Override Seed=false
    Velocity Mode=::NewEnumerator1
    Velocity Origin=$Particles.ShapeLocation.ShapePosition
    Velocity Speed=&Multiply_Float(A=200, B=0.13)
   ) {module=AddVelocity001}
   &CameraOffset(Camera Offset Amount=10)
  ]
  particleUpdateScript [
   &ParticleState()
   &ScaleSpriteSize(Scale Sprite Size Mode=::NewEnumerator0, Uniform Scale Factor=0.15) {module=ScaleSpriteSize001}
   &GravityForce(Gravity=&Multiply_VectorByFloat(Vector=<0 0 -200>, Float=0.7)) {module=GravityForce001}
   &Drag(Rotational Drag=50, Use Linear Drag=true, Use Rotational Drag=false)
   &SolveForcesAndVelocity(Clamp Velocity=false, Limit Acceleration=false)
   &SpriteRotationRate(Rotation Rate=&FloatFromCurve(
     FloatCurve=DataInterface() {propertyValues {Curve {keys [<at 0 1>,<at 1 0>]}}}
     Scale Curve=&MakeCustomFloatFromBool(Bool=&RandomBool(), True Float=-90, False Float=90)
    ))
   &ScaleSpriteSize(Scale Factor=&Vector2DFromCurve(
     Vector2Curve=DataInterface() {propertyValues {
       XCurve {keys [{arriveTangent=0.5278, interpMode=RCIM_Cubic, time=0, value=0.5}, {arriveTangent=1.8113, tangentMode=RCTM_User, time=0.4, value=1}]}
       YCurve {keys [
         {arriveTangent=1.4008, interpMode=RCIM_Cubic, leaveTangent=1.4008, tangentMode=RCTM_User, time=0, value=0.2}
         {interpMode=RCIM_Cubic, tangentMode=RCTM_User, time=0.75, value=1}
        ]}
      }}
     ScaleCurve=<1 1.3>
    ))
   &ScaleSpriteSizeBySpeed(Max Scale Factor=<1.2 1.5>, Velocity Threshold=800)
   &ScaleColor(
    Scale Alpha=&FloatFromCurve(FloatCurve=DataInterface() {propertyValues {Curve {keys [{interpMode=RCIM_Cubic, tangentMode=RCTM_User, time=0.9, value=1}, {interpMode=RCIM_Cubic, tangentMode=RCTM_User, time=1, value=0}]}}})
    ScaleA=true
    ScaleRGB=true
   )
   &DynamicMaterialParameters(
    Index 0 Param 2=&FloatFromCurve(
     FloatCurve=DataInterface() {propertyValues {Curve {keys [{arriveTangent=-1.9579, interpMode=RCIM_Cubic, time=0.1, value=0.2}, {interpMode=RCIM_Cubic, tangentMode=RCTM_User, time=0.3, value=0}]}}}
     Scale Curve=0.5
    )
    Index 0 Param 4=&FloatFromCurve(
     FloatCurve=DataInterface() {propertyValues {Curve {keys [
         {arriveTangent=1.2665, interpMode=RCIM_Cubic, leaveTangent=1.2665, tangentMode=RCTM_User, time=0.1, value=0}
         {arriveTangent=1.0703, interpMode=RCIM_Cubic, leaveTangent=1.0703, tangentMode=RCTM_User, time=0.75, value=2}
        ]}}}
     Scale Curve=&RandomRangeFloat(Minimum=0.5, Maximum=0.8)
    )
    Module.Param0WriteEnabled=false
    Module.Param1WriteEnabled=true
    Module.Param2WriteEnabled=false
    Module.Param3WriteEnabled=true
   )
  ]
  rendererClasses [&NiagaraSpriteRendererProperties]
  renderers [
   {rendererClass=&NiagaraSpriteRendererProperties}
   {rendererClass=&NiagaraSpriteRendererProperties, rendererIndex=1}
  ]
 }
}
summary {
 systemName=NS_Emote_VowelSleep_BallIdle
 emitters [
  {emitterName=Smoke, rendererClasses [&NiagaraSpriteRendererProperties]}
  {emitterName=SmokeLargeLocalSpace, rendererClasses [&NiagaraSpriteRendererProperties]}
  {emitterName=Splash_Hard, rendererClasses [&NiagaraSpriteRendererProperties]}
 ]
 userVariables [
  User.CustomFloatType {type=NiagaraFloat, defaultValue=0, description=""}
  User.Mesh Int {type=NiagaraInt32}
 ]
}
