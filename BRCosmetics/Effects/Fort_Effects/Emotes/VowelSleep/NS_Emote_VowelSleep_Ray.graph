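# Niagara system for the VowelSleep emote that hangs a spinning disco ball light show about 230 units up.
# Burst-spawned emitters build the scaling-in ball, color-cycling rays with a dynamic light, and a rotating sphere of glow sprites.
# Rate-spawned mesh beams, spotlight and sparkles flicker through color curves, with GPU sim for sharp rays and one sparkle layer.
# System RandomStartTiming scales most alphas; some intensities also vary by front end or time of day.
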NiagaraSystem /BRCosmetics/Effects/Fort_Effects/Emotes/VowelSleep/NS_Emote_VowelSleep_Ray
dependencies {
 usedDataInterfaces [
  &NiagaraDataInterfaceArrayFloat3
  &NiagaraDataInterfaceColorCurve
  &NiagaraDataInterfaceCurve
  &NiagaraDataInterfaceArrayFloat
  &NiagaraDataInterfaceVectorCurve
 ]
 usedDynamicInputs []
 usedModules [
  &SystemState
  &self:SystemSpawnScript.SystemScriptSource.SystemScriptGraph.NiagaraNodeAssignment_0.SetVariables_#1
 ]
}
emitters {
 BlueLight_MESH {
  emitterUpdateScript [
   &EmitterState()
   &SpawnRate(SpawnRate=6, Use Spawn Probability=false)
  ]
  particleSpawnScript [
   &InitializeParticle(Mesh Scale=<0.62 0.62 2>, Mesh Scale Mode=::NewEnumerator4, Position Offset=<0 0 250>, UsePositionOffset=true, Write Material Random=true)
   &CameraOffset(Camera Offset Amount=-30)
   &InitialMeshOrientation(
    Mesh Orientation Mode=::NewEnumerator7
    Rotation=&RandomRangeVector(Minimum=<-0.0417 -0.0417 -0.0417>, Maximum=<0.0417 0.0417 0.0417>)
    Rotation Coordinate Space=::NewEnumerator2
    Use Rotation Vector=true
   )
  ]
  particleUpdateScript [
   &ParticleState()
   &UpdateMeshOrientation(
    Rotation Coordinate Space=::NewEnumerator2
    Rotation Rate=&RandomRangeFloat(Minimum=-0.3, Maximum=0.3)
    Rotation Vector=<0 0 1>
   )
   &ScaleColor(
    Scale Alpha=&Multiply_Float(
     A=&FloatFromCurve(
      FloatCurve=DataInterface() {propertyValues {Curve {keys [<at 0 0>,<at 0.15 0.7>,<at 0.27 0>,<at 0.38 0.7>,<at 0.64 0>,<at 0.85 0.7>,<at 1 0>]}}}
      Scale Curve=0.7
     )
     B=$System.RandomStartTiming
    )
    Scale RGB=&VectorFromCurve(VectorCurve=DataInterface() {propertyValues {
       XCurve {keys [<at 0.35 0>,<at 0.4 0.5>,<at 0.66 0.5>,<at 0.7 0>,<at 0.9 0>]}
       YCurve {keys [<at 0.35 0.04>,<at 0.4 0>,<at 0.66 0>,<at 0.7 0.04>,<at 0.88 0.04>]}
       ZCurve {keys [<at 0.35 0.5>,<at 0.4 0>,<at 0.66 0>,<at 0.7 0.5>,<at 0.88 0.5>]}
      }})
    ScaleA=true
    ScaleRGB=true
   )
  ]
  rendererClasses [&NiagaraMeshRendererProperties]
  renderers [{rendererClass=&NiagaraMeshRendererProperties, rendererIndex=0}]
 }
 BlueLight_thin {
  emitterUpdateScript [
   &EmitterState(
    Inactive Response=::NewEnumerator1
    Life Cycle Mode=::NewEnumerator1
    Loop Delay=1
    Loop Duration=4.8
    Recalculate Duration Each Loop=true
    UseLoopDelay=true
   )
   &SpawnRate(SpawnRate=&RandomRangeFloat(Minimum=5, Maximum=10), Use Spawn Probability=false)
  ]
  particleSpawnScript [
   &FN_SetSpawnGroup(Number of Spawn Groups=6)
   &InitializeParticle(
    Lifetime Max=0.6
    Lifetime Min=0.6
    Lifetime Mode=::NewEnumerator1
    Mesh Scale=&Multiply_Vector(
     A=&SelectVectorFromArray(
      Direct Array Index=$Particles.SpawnGroup
      Vector Selection Array=DataInterface() {propertyValues {FloatData [<0.1 0.1 1.5>, <0.15 0.15 1.7>, <0.08 0.1 1.9>, <0.15 0.15 1.85>, <0.1 0.07 2>, <0.15 0.15 1.7>]}}
     )
     B=<1 1 0.75>
    )
    Mesh Scale Mode=::NewEnumerator4
    Position Offset=&AddVectorToPosition(
     Position=&Multiply_VectorByFloat(
      Vector=&SelectVectorFromArray(
       Direct Array Index=$Particles.SpawnGroup
       Vector Selection Array=DataInterface() {propertyValues {FloatData [<30 20 0>, <30 -20 0>, <0 35 0>, <0 -35 0>, <-40 20 0>, <-40 -20 0>]}}
      )
      Float=-0.2
     )
     Vector=<0 0 230>
    )
    UsePositionOffset=true
    Write Material Random=true
   )
   &InitialMeshOrientation(
    Mesh Orientation Mode=::NewEnumerator2
    Rotation=&Multiply_VectorByFloat(
     Vector=&SelectVectorFromArray(
      Direct Array Index=$Particles.SpawnGroup
      Vector Selection Array=DataInterface() {propertyValues {FloatData [<0.13 0.5 -0.1>, <0.17 0.5 0.1>, <0 0.5 -0.1>, <0 0.5 0.1>, <0.05 0.45 -0.05>, <0 0.45 0.05>]}}
     )
     Float=1
    )
    Use Rotation Vector=true
   )
  ]
  particleUpdateScript [
   &ParticleState()
   &Color(Color=&ColorFromCurve(ColorCurve=DataInterface() {propertyValues {
       AlphaCurve {keys [<at 0 0.75 RCIM_Cubic>]}
       BlueCurve {keys [<at 0.3233 0.5 RCIM_Cubic>,<at 0.3778 0.5>,<at 0.65 0.5>,<at 0.7 0.3531>]}
       GreenCurve {keys [<at 0.3233 0.4912 RCIM_Cubic>,<at 0.3778 0.1788>,<at 0.65 0.0992>,<at 0.7 0.064>]}
       RedCurve {keys [<at 0.3233 0.0609 RCIM_Cubic>,<at 0.3778 0.0412>,<at 0.65 0.1414>,<at 0.7 0.5>]}
      }})) {module=Color001}
   &ScaleColor(
    Color Value To Scale=$Particles.Color
    Scale Alpha=&Multiply_Float(
     A=&FloatFromCurve(
      FloatCurve=DataInterface() {propertyValues {Curve {keys [
          {arriveTangent=6.8924, interpMode=RCIM_Cubic, leaveTangent=6.8924, tangentMode=RCTM_User, time=0, value=0}
          {interpMode=RCIM_Cubic, tangentMode=RCTM_User, time=0.25, value=1}
          {tangentMode=RCTM_User, time=1, value=0}
         ]}}}
      Scale Curve=&Multiply_Float(A=0.2, B=$System.RandomStartTiming)
     )
     B=&SelectFloatFromArray(Direct Array Index=$Particles.SpawnGroup, Float Selection Array=DataInterface() {propertyValues {FloatData [1,1,1,1,1,1]}})
    )
    Scale RGB=&VectorFromFloat(Value=&FloatFromCurve(FloatCurve=DataInterface() {propertyValues {Curve {keys [<at 0 0>,<at 0.2 10>,<at 0.3 0.5>,<at 0.4 12>,<at 0.6 2>,<at 0.7 8.5>,<at 1 0>]}}}))
    ScaleA=true
    ScaleRGB=true
   )
   &ScaleMeshSize(Scale Factor=&MakeVector(
     X=1
     Y=1
     Z=&FloatFromCurve(FloatCurve=DataInterface() {propertyValues {Curve {keys [
          {interpMode=RCIM_Cubic, leaveTangent=1.3994, tangentMode=RCTM_Break, time=0, value=0.5}
          {interpMode=RCIM_Cubic, tangentMode=RCTM_Break, time=1, value=1}
         ]}}})
    ))
   &UpdateMeshOrientation(
    Rotation Coordinate Space=::NewEnumerator0
    Rotation Rate=&Multiply_Float(
     A=&SelectFloatFromArray(Direct Array Index=$Particles.SpawnGroup, Float Selection Array=DataInterface() {propertyValues {FloatData [0.5,-0.6,0.45,-0.48,0.6,-0.55]}})
     B=&RandomRangeFloat(Minimum=-0.3, Maximum=0.3)
    )
    Rotation Vector=<0 0 1>
   )
   &CameraOffset(Camera Offset Amount=-30)
  ]
  rendererClasses [&NiagaraMeshRendererProperties]
  renderers [{rendererClass=&NiagaraMeshRendererProperties, rendererIndex=0}]
 }
 DiscoBall {
  emitterUpdateScript [
   &EmitterState(
    Inactive Response=::NewEnumerator1
    Life Cycle Mode=::NewEnumerator1
    Loop Behavior=::NewEnumerator1
    Loop Duration Mode=::NewEnumerator1
    UseLoopDelay=false
   )
   &SpawnBurst_Instantaneous(Use Spawn Probability=false, UseLoopCountLimit=false)
  ]
  particleSpawnScript [&InitializeParticle(Color Mode=::NewEnumerator0, Mesh Scale Mode=::NewEnumerator3, Position Offset=<0 0 230>, UsePositionOffset=true, Write Material Random=true)]
  particleUpdateScript [
   &ParticleState(Kill Particles When Lifetime Has Elapsed=false)
   &UpdateMeshOrientation(Rotation Rate=0.3, Rotation Vector=<0 0 1>)
   &ScaleMeshSize(Scale Factor=&Multiply_VectorByFloat(
     Vector=<1 1 1>
     Float=&FloatFromCurve(FloatCurve=DataInterface() {propertyValues {Curve {keys [<at 0.1 0>,<at 0.4 1>,<at 3 1>]}}}, CurveIndex=$Emitter.Age)
    ))
  ]
  rendererClasses [&NiagaraMeshRendererProperties]
  renderers [{rendererClass=&NiagaraMeshRendererProperties, rendererIndex=0}]
 }
 DiscoBallRaySharp {
  simTarget=GPUComputeSim
  emitterUpdateScript [
   &EmitterState(Delay First Loop Only=true, Life Cycle Mode=::NewEnumerator1, Loop Delay=1, Loop Duration=4.8, UseLoopDelay=true)
   &SpawnRate(SpawnRate=15, Use Spawn Probability=false)
   &SpawnBurst_Instantaneous(Spawn Count=10, Use Spawn Probability=false)
  ]
  particleSpawnScript [
   &InitializeParticle(
    Color=&RandomRangeLinearColor(Minimum=<rgba 5 0 1.176 3>, Maximum=<rgba 0 5 3.161 3>)
    Lifetime Max=0.2
    Lifetime Min=0.1
    Lifetime Mode=::NewEnumerator1
    Mesh Scale Max=<1.2 0.2 0.2>
    Mesh Scale Min=<1.2 0.2 0.2>
    Mesh Scale Mode=::NewEnumerator2
    UsePositionOffset=false
    Write Material Random=true
   ) {module=InitializeParticle001}
   &DynamicMaterialParameters(
    Index 0 Param 1=&UniformRangedFloat(Minimum=-1, Maximum=1)
    Index 0 Param 2=1
    Module.Param0WriteEnabled=true
    Module.Param1WriteEnabled=true
    Module.Param2WriteEnabled=true
    Module.Param3WriteEnabled=true
   )
   &InitialMeshOrientation(
    Orientation Vector=<1 0 1>
    Rotation=&Add_Vector(A=<0 0 0>, B=&RandomRangeVector(Minimum=<-1 0 -1>, Maximum=<1 1 1>))
    Use Orientation Vector=true
    Use Rotation Vector=true
   ) {module=InitialMeshOrientation001}
   &SystemLocation(MaskSpawnGroup=false, Offset=<0 0 230>)
  ]
  particleUpdateScript [
   &CameraOffset(Camera Offset Amount=-30)
   &ParticleState()
   &SolveForcesAndVelocity(Clamp Velocity=false, Limit Acceleration=false)
   &ScaleColor(
    Scale Alpha=&Multiply_Float(
     A=&FloatFromCurve(
      FloatCurve=DataInterface() {propertyValues {Curve {keys [
          {interpMode=RCIM_Cubic, tangentMode=RCTM_User, time=0, value=0}
          {arriveTangent=-1.6906, interpMode=RCIM_Cubic, leaveTangent=-1.6906, tangentMode=RCTM_User, time=0.1, value=0.5}
          {tangentMode=RCTM_User, time=1, value=0}
         ]}}}
      Scale Curve=&FloatFromCurve(
       FloatCurve=DataInterface() {propertyValues {Curve {keys [<at 0 0>,<at 0.3005 1>,<at 1 1>]}}}
       Scale Curve=&Multiply_Float(A=&FN_FrontEndFloatModifier(Gameplay=1, Front End=0.5), B=$System.RandomStartTiming)
      )
     )
     B=&RandomRangeFloat(Minimum=1, Maximum=2.2)
    )
    Scale RGB=&Multiply_VectorByFloat(
     Vector=<1 1 1>
     Float=&FNTimeOfDayScalar(Night Scale=0.5, Morning Scale=0.3, Day Scale=0.7, Evening Scale=0.4)
    )
    ScaleA=true
    ScaleRGB=true
   )
   &ScaleMeshSize(Scale Factor=&MakeVector(
     X=&FloatFromCurve(FloatCurve=DataInterface() {propertyValues {Curve {keys [<at 0 0.8>,<at 1 1.2>]}}}, Scale Curve=0.8)
     Y=1
     Z=&FloatFromCurve(FloatCurve=DataInterface() {propertyValues {Curve {keys [<at 0 1.8>,<at 1 1>]}}})
    ))
  ]
  rendererClasses [&NiagaraMeshRendererProperties]
  renderers [{rendererClass=&NiagaraMeshRendererProperties, rendererIndex=0}]
 }
 DiscoBallRays {
  emitterUpdateScript [
   &EmitterState(Life Cycle Mode=::NewEnumerator1, Loop Behavior=::NewEnumerator1, Loop Delay=0.7, Loop Duration Mode=::NewEnumerator1, UseLoopDelay=true)
   &SpawnBurst_Instantaneous(Use Spawn Probability=false, UseLoopCountLimit=false)
  ]
  particleSpawnScript [
   &InitializeParticle(
    Color=<rgba 0.5 0.1919 0.142 1>
    Mesh Scale Mode=::NewEnumerator3
    Mesh Uniform Scale=1.5
    Position Offset=<0 0 230>
    UsePositionOffset=true
    Write Material Random=true
   )
   &DynamicMaterialParameters(
    Index 0 Param 1=&RandomRangeFloat(Minimum=0, Maximum=10)
    Index 0 Param 2=&RandomRangeFloat(Minimum=10, Maximum=10)
    Index 0 Param 3=&RandomRangeFloat(Minimum=0.8, Maximum=1.6)
    Module.Index0Param0Write=true
    Module.Index0Param1Write=true
    Module.Index0Param2Write=true
    Module.Index0Param3Write=false
   )
   &InitialMeshOrientation(Mesh Orientation Mode=::NewEnumerator7, Rotation=<0.15 0 0>, Use Rotation Vector=true)
  ]
  particleUpdateScript [
   &ParticleState(Kill Particles When Lifetime Has Elapsed=false, Loop Particles Lifetime=true)
   &ScaleMeshSize()
   &ScaleColor(
    Scale Alpha=&FloatFromCurve(
     FloatCurve=DataInterface() {propertyValues {Curve {keys [<at 0 1>,<at 0.7 1>,<at 0.9 1>,<at 2 1>,<at 5.8 1>,<at 6.1 1>]}}}
     CurveIndex=$Emitter.Age
     Scale Curve=&Multiply_Float(A=&FN_FrontEndFloatModifier(Gameplay=2, Front End=0.5), B=$System.RandomStartTiming)
    )
    ScaleA=true
    ScaleRGB=true
   )
   &CameraOffset()
   /SetVariables(Particles.Color003:LinearColor=&MakeLinearColorFromVectorAndFloat(
     "Vector (RGB)"=&MakeVectorFromLinearColorRGB(Color=&Multiply_LinearColorByFloat(
       Linear Color=&ColorFromCurve(ColorCurve=DataInterface() {propertyValues {
          AlphaCurve {keys []}
          BlueCurve {keys [<at 0 0.016 RCIM_Cubic>,<at 0.2551 1>,<at 0.5364 0.9516>,<at 0.7374 2>,<at 1 0.7988 RCIM_Cubic>]}
          GreenCurve {keys [<at 0 0 RCIM_Cubic>,<at 0.2551 0>,<at 0.5364 1>,<at 0.7374 0.1718>,<at 1 0 RCIM_Cubic>]}
          RedCurve {keys [<at 0 1 RCIM_Cubic>,<at 0.2551 0.4094>,<at 0.5364 0>,<at 0.7374 0.4103>,<at 1 1 RCIM_Cubic>]}
         }})
       Float=0.3
      ))
     "Float (Alpha)"=&FloatFromCurve(
      FloatCurve=DataInterface() {propertyValues {Curve {keys [<at 0 1>,<at 0.7 1>,<at 0.9 1>,<at 2 1>,<at 5.8 1>,<at 6.1 1>]}}}
      CurveIndex=$Emitter.Age
      Scale Curve=&Multiply_Float(
       A=1.3
       B=&Multiply_Float(A=&FN_FrontEndFloatModifier(Gameplay=1.3, Front End=0.3), B=$System.RandomStartTiming)
      )
     )
    )) {isSetParameters=true}
   /SetVariables(Particles.Size003:Vector2f=&Multiply_Vector2D(A=&Vector2DFromFloat(Value=200), B=<1 1>)) {isSetParameters=true}
   /SetVariables(Particles.Dyn_glow:Vector4f=&MakeVector4(X=2.5, Y=5, Z=0, W=0)) {isSetParameters=true}
   &Light_Attributes(
    Color=1
    Color Mode=0
    Falloff Exponent=2
    Light Color:LinearColor=&Multiply_LinearColorByFloat(
     Linear Color=&ColorFromCurve(ColorCurve=DataInterface() {propertyValues {
        AlphaCurve {keys []}
        BlueCurve {keys [<at 0 0.016 RCIM_Cubic>,<at 0.2551 1>,<at 0.5364 0.9516>,<at 0.7374 2>,<at 1 0.7988 RCIM_Cubic>]}
        GreenCurve {keys [<at 0 0 RCIM_Cubic>,<at 0.2551 0>,<at 0.5364 1>,<at 0.7374 0.1718>,<at 1 0 RCIM_Cubic>]}
        RedCurve {keys [<at 0 1 RCIM_Cubic>,<at 0.2551 0.4094>,<at 0.5364 0>,<at 0.7374 0.4103>,<at 1 1 RCIM_Cubic>]}
       }})
     Float=&Multiply_Float(
      A=10
      B=&FloatFromCurve(
       FloatCurve=DataInterface() {propertyValues {Curve {keys [<at 0 1>,<at 0.7 1>,<at 0.9 1>,<at 2 1>,<at 5.8 1>,<at 6.1 1>]}}}
       CurveIndex=$Emitter.Age
       Scale Curve=&Multiply_Float(A=&FN_FrontEndFloatModifier(Gameplay=1, Front End=0.5), B=$System.RandomStartTiming)
      )
     )
    )
    Light Exponent=3
    Light Position Offset:Vector3f=&MakeVector(
     X=0
     Y=&FN_LOD_Distance_Remap(Start Distance=300, End Distance=400, Minimum Output=-30, Maximum Output=30, Clamp Output=true)
     Z=-100
    )
    Position=1
    Radius=2
    UsePositionOffset=true
   )
  ]
  rendererClasses [
   &NiagaraMeshRendererProperties
   &NiagaraSpriteRendererProperties
   &NiagaraLightRendererProperties
  ]
  renderers [
   {rendererClass=&NiagaraMeshRendererProperties, rendererIndex=0}
   {rendererClass=&NiagaraSpriteRendererProperties, rendererIndex=1}
   {rendererClass=&NiagaraLightRendererProperties, rendererIndex=2}
  ]
 }
 ShapeLocationGlow {
  emitterUpdateScript [
   &EmitterState(Life Cycle Mode=::NewEnumerator1, Loop Behavior=::NewEnumerator1, Loop Delay=1, Loop Duration Mode=::NewEnumerator1, UseLoopDelay=true)
   &SpawnBurst_Instantaneous(Spawn Count=150, Use Spawn Probability=false, UseLoopCountLimit=false)
  ]
  particleSpawnScript [
   /SetVariables(Particles.RandomCOlor:LinearColor=&RandomRangeLinearColor(Minimum=<rgba 0.0003 0.3008 1 1>, Maximum=<rgba 1 0.0185 0.7989 1>, Color Channel Mode=::NewEnumerator2)) {isSetParameters=true}
   /SetVariables(Particles.SizeAll=2) {isSetParameters=true}
   &InitializeParticle(
    AdjustAlpha=false
    AdjustHue=true
    AdjustSaturation=false
    AdjustValue=false
    Color=$Particles.RandomCOlor
    Color Mode=::NewEnumerator3
    Hue Shift Range:Vector2f=<-1 1>
    Lifetime Mode=::NewEnumerator1
    Sprite Size Mode=::NewEnumerator3
    Uniform Sprite Size=1.5
    UsePositionOffset=false
    Write Material Random=true
   )
   &ApplyInitialForces()
   &DynamicMaterialParameters(
    Index 0 Param 2=3
    Index 1 Param 1=3
    Index 1 Param 2=3
    Module.Index0Float0Write=true
    Module.Index0Float1Write=true
    Module.Index0Float2Write=true
    Module.Index0Float3Write=true
    Module.Index1Float0Write=true
    Module.Index1Float1Write=true
    Module.Index1Float2Write=true
    Module.Index1Float3Write=true
    Write Parameter Index 1=::NewEnumerator1
   )
  ]
  particleUpdateScript [
   &SolveForcesAndVelocity(Clamp Velocity=false, Limit Acceleration=false)
   &ScaleColor(
    Scale Alpha=&Multiply_Float(
     A=0.5
     B=&FloatFromCurve(
      FloatCurve=DataInterface() {propertyValues {Curve {keys [<at 0 1>,<at 0.7 1>,<at 0.9 1>,<at 2 1>,<at 5.5 1>,<at 6 1>]}}}
      CurveIndex=$Emitter.Age
      Scale Curve=&Multiply_Float(A=&FN_FrontEndFloatModifier(Gameplay=1, Front End=0.6), B=$System.RandomStartTiming)
     )
    )
    Scale RGB=&Multiply_VectorByFloat(
     Vector=<1 1 1>
     Float=&Multiply_Float(
      A=&Multiply_Float(A=3, B=1)
      B=&RandomRangeFloat(Minimum=0.5, Maximum=1, Evaluation Type=::NewEnumerator1)
     )
    )
    ScaleA=true
    ScaleRGB=true
   )
   &ShapeLocation(
    MaskSpawnGroup=false
    Offset=<0 0 230>
    Offset Mode=::NewEnumerator0
    Rotation Angle:NiagaraFloat=&Multiply_Float(A=0.1, B=$Engine.Time)
    Rotation Axis=<0 0 1>
    Rotation Mode=::NewEnumerator1
    Sphere Distribution=::NewEnumerator3
    Sphere Radius=40
    UseOwnerScale=true
   )
   &AlignSpriteToMeshOrientation(Mesh Orientation Relative Sprite Facing Vector:Vector3f=$Particles.ShapeLocation.ShapeNormal, SetAlignment=true, SetFacing=true)
   &ParticleState(Kill Particles When Lifetime Has Elapsed=false, Loop Particles Lifetime=true)
   &ScaleSpriteSize(Scale Sprite Size Mode=::NewEnumerator0, Uniform Scale Factor=$Particles.SizeAll)
   &FN_Sprite_Attributes(
    Color=&MakeLinearColorFromVectorAndFloat(
     "Vector (RGB)"=&MakeVectorFromLinearColorRGB(Color=&Multiply_LinearColorByFloat(
       Linear Color=$Particles.RandomCOlor
       Float=&Multiply_Float(
        A=&Multiply_Float(
         A=2
         B=&FloatFromCurve(FloatCurve=DataInterface() {propertyValues {Curve {keys [<at 0 0>,<at 0.3 1>,<at 0.9 1>,<at 1 1>]}}})
        )
        B=&RandomRangeFloat(Minimum=0.5, Maximum=1.3, Evaluation Type=::NewEnumerator1)
       )
      ))
     "Float (Alpha)"=&Multiply_Float(
      A=0.5
      B=&FloatFromCurve(
       FloatCurve=DataInterface() {propertyValues {Curve {keys [<at 0 1>,<at 0.7 1>,<at 0.9 1>,<at 2 1>,<at 5.5 1>,<at 6 1>]}}}
       CurveIndex=$Emitter.Age
       Scale Curve=&Multiply_Float(A=&FN_FrontEndFloatModifier(Gameplay=0.8, Front End=0.6), B=$System.RandomStartTiming)
      )
     )
    )
    Create Color Attribute=true
    Create SpriteSize Attribute=true
    Module Output:ENiagara_AttributeSamplingApplyOutput=::NewEnumerator1
    SpriteSize=&Vector2DFromFloat(Value=&Multiply_Float(A=12, B=$Particles.SizeAll))
   )
  ]
  rendererClasses [&NiagaraSpriteRendererProperties]
  renderers [
   {rendererClass=&NiagaraSpriteRendererProperties}
   {rendererClass=&NiagaraSpriteRendererProperties, rendererIndex=1}
  ]
 }
 Sparklys {
  simTarget=GPUComputeSim
  emitterUpdateScript [
   &EmitterState()
   &SpawnRate(SpawnRate=&RandomRangeFloat(Minimum=5, Maximum=10), Use Spawn Probability=false)
  ]
  particleSpawnScript [
   &InitializeParticle(
    Color=<rgba 0.5928 0.5889 1 20>
    Lifetime Max=0.2
    Lifetime Min=0.2
    Lifetime Mode=::NewEnumerator1
    Mesh Scale=<1 1 0.8>
    Mesh Scale Mode=::NewEnumerator4
    Position Offset=<0 0 150>
    Sprite Rotation Mode=::NewEnumerator2
    Sprite Size Mode=::NewEnumerator1
    Uniform Sprite Size Max=40
    Uniform Sprite Size Min=10
    UsePositionOffset=true
    Write Material Random=true
   )
   &ShapeLocation(MaskSpawnGroup=false, Override Seed=false, Sphere Radius=50, UseOwnerScale=true)
   &CameraOffset(Camera Offset Amount=50)
  ]
  particleUpdateScript [
   &ParticleState()
   &ScaleSpriteSize(Scale Sprite Size Mode=::NewEnumerator0, Uniform Scale Factor=0.5)
   &ScaleColor(
    Scale Alpha=&Multiply_Float(
     A=&FloatFromCurve(FloatCurve=DataInterface() {propertyValues {Curve {keys [<at 0 0>,<at 0.14 1>,<at 0.38 0>,<at 0.52 0.75>,<at 1 0.75>]}}})
     B=$System.RandomStartTiming
    )
    Scale RGB=&VectorFromCurve(VectorCurve=DataInterface() {propertyValues {
       XCurve {keys [<at 0.25 1>,<at 0.3 0>,<at 0.6 0>,<at 0.7 1>,<at 0.9 1>]}
       YCurve {keys [<at 0.25 0.1>,<at 0.3 0.08>,<at 0.65 0.5>,<at 0.7 0.1>,<at 0.88 0.1>]}
       ZCurve {keys [<at 0.25 0.1>,<at 0.3 1>,<at 0.65 0.04>,<at 0.7 0.1>,<at 0.88 0.1>]}
      }})
    ScaleA=true
    ScaleRGB=true
   )
  ]
  rendererClasses [&NiagaraSpriteRendererProperties]
  renderers [{rendererClass=&NiagaraSpriteRendererProperties}]
 }
 Sparklys001 {
  emitterUpdateScript [
   &EmitterState()
   &SpawnRate(SpawnRate=30, Use Spawn Probability=false)
  ]
  particleSpawnScript [
   &InitializeParticle(
    Color Maximum=<rgba 0.3 0.4167 1 1>
    Color Minimum=<rgba 1 0.3 1 1>
    Color Mode=::NewEnumerator2
    Lifetime Max=0.2
    Lifetime Min=0.15
    Lifetime Mode=::NewEnumerator1
    Mesh Scale=<1 1 0.8>
    Mesh Scale Mode=::NewEnumerator4
    Position Offset=<0 0 230>
    Sprite Rotation Mode=::NewEnumerator2
    Sprite Size Mode=::NewEnumerator1
    Uniform Sprite Size Max=150
    Uniform Sprite Size Min=70
    UsePositionOffset=true
    Write Material Random=true
   )
   &ShapeLocation(
    MaskSpawnGroup=false
    Override Seed=false
    Sphere Radius=&RandomRangeFloat(Minimum=0, Maximum=50)
    UseOwnerScale=true
   )
   &CameraOffset(Camera Offset Amount=40)
  ]
  particleUpdateScript [
   &ParticleState()
   &ScaleSpriteSize(
    Uniform Curve Scale=0.5
    Uniform Curve Sprite Scale=DataInterface() {propertyValues {Curve {
       defaultValue=0
       keys [{interpMode=RCIM_Cubic, tangentMode=RCTM_User, time=0, value=1}, {interpMode=RCIM_Cubic, tangentMode=RCTM_User, time=1, value=0}]
      }}}
   )
   &ScaleColor(
    Scale Alpha=&Multiply_Float(
     A=&FloatFromCurve(FloatCurve=DataInterface() {propertyValues {Curve {keys [<at 0 0>,<at 0.15 1>,<at 0.38 0>,<at 0.52 0.75>,<at 1 0.75>]}}})
     B=$System.RandomStartTiming
    )
    ScaleA=true
    ScaleRGB=true
   )
  ]
  rendererClasses [&NiagaraSpriteRendererProperties]
  renderers [{rendererClass=&NiagaraSpriteRendererProperties}]
 }
 SpotLight_Main {
  emitterUpdateScript [
   &EmitterState()
   &SpawnRate(SpawnRate=2, Use Spawn Probability=false)
  ]
  particleSpawnScript [
   &InitializeParticle(
    Color=<rgba 0.3072 0.0782 1 2>
    Lifetime=0.5
    Mesh Scale=<1 1 0.8>
    Mesh Scale Mode=::NewEnumerator4
    Position Offset=<0 0 210>
    UsePositionOffset=true
    Write Material Random=true
   )
   &CameraOffset(Camera Offset Amount=-30)
  ]
  particleUpdateScript [
   &ParticleState()
   &ScaleColor(
    Scale Alpha=&Multiply_Float(
     A=&FloatFromCurve(FloatCurve=DataInterface() {propertyValues {Curve {keys [<at 0 0.4>,<at 0.37 0.4>,<at 0.58 0.1>,<at 0.7878 0.4>,<at 1 0.4>]}}})
     B=$System.RandomStartTiming
    )
    Scale RGB=&VectorFromCurve(VectorCurve=DataInterface() {propertyValues {
       XCurve {keys [<at 0 4>,<at 0.3 0.8>,<at 0.53 0.3>,<at 0.68 0.3>,<at 0.9 4>]}
       YCurve {keys [<at 0.12 0.4>,<at 0.3 1.06>,<at 0.41 0.88>,<at 0.5346 2>,<at 0.68 2>,<at 0.88 0.4>]}
       ZCurve {keys [<at 0.12 0.4>,<at 0.3 4>,<at 0.41 4>,<at 0.53 0.725>,<at 0.68 0.725>,<at 0.88 0.4>]}
      }})
    ScaleA=true
    ScaleRGB=true
   )
  ]
  rendererClasses [&NiagaraMeshRendererProperties]
  renderers [
   {rendererClass=&NiagaraMeshRendererProperties, rendererIndex=0}
   {rendererClass=&NiagaraMeshRendererProperties, rendererIndex=1}
  ]
 }
}
summary {
 systemName=NS_Emote_VowelSleep_Ray
 emitters [
  {emitterName=BlueLight_thin, rendererClasses [&NiagaraMeshRendererProperties]}
  {
   emitterName=DiscoBallRays
   rendererClasses [
    &NiagaraMeshRendererProperties
    &NiagaraSpriteRendererProperties
    &NiagaraLightRendererProperties
   ]
  }
  {emitterName=DiscoBallRaySharp, simTarget=GPUComputeSim, rendererClasses [&NiagaraMeshRendererProperties]}
  {emitterName=ShapeLocationGlow, rendererClasses [&NiagaraSpriteRendererProperties]}
  {emitterName=DiscoBall, rendererClasses [&NiagaraMeshRendererProperties]}
  {emitterName=SpotLight_Main, rendererClasses [&NiagaraMeshRendererProperties]}
  {emitterName=Sparklys, simTarget=GPUComputeSim, rendererClasses [&NiagaraSpriteRendererProperties]}
  {emitterName=BlueLight_MESH, rendererClasses [&NiagaraMeshRendererProperties]}
  {emitterName=Sparklys001, rendererClasses [&NiagaraSpriteRendererProperties]}
 ]
 userVariables [
  User.bIsFrontEnd {type=NiagaraBool}
  User.bIsFrontEndPreview {type=NiagaraBool}
 ]
}
